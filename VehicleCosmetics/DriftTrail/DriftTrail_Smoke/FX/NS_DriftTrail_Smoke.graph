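# Niagara system for the Smoke drift trail vehicle cosmetic, emitting tinted smoke from vehicle skeletal mesh sockets.
# MainTrail and SmokeWheel spawn rearward-moving sprites with curve-driven size, alpha and material parameters; MainTrail varies rate, lifetime and velocity by game context.
# WheelWrap bursts two mesh particles that re-sample their sockets each update, with per-wheel scaling and tire-FX disable logic.
# All three emitters partition particles across sockets and take painted-variant colors.
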NiagaraSystem /VehicleCosmetics/DriftTrail/DriftTrail_Smoke/FX/NS_DriftTrail_Smoke
dependencies {
 usedDataInterfaces [
  &NiagaraDataInterfaceCurve
  &NiagaraDataInterfaceSpriteRendererInfo
  &NiagaraDataInterfaceSkeletalMesh
 ]
 usedDynamicInputs []
 usedModules [
  &self:SystemSpawnScript.SystemScriptSource.SystemScriptGraph.NiagaraNodeAssignment_0.SetVariables_#1
  &self:SystemSpawnScript.SystemScriptSource.SystemScriptGraph.NiagaraNodeAssignment_1.SetVariables_#2
  &SystemState
  &FN_UnpackVehicleFlags
  /Game/Effects/Niagara/Modules/Vehicle/FN_UpdateVehicleAttribs
  /VehicleCosmetics/FX/Niagara/Modules/VC_SetVehicleSpeedAttribs
  /Game/Effects/Niagara/Modules/Vehicle/FN_VehicleFlags_Cheatcodes
 ]
}
emitters {
 NE_CS_MainTrail {
  emitterUpdateScript [
   &EmitterState()
   &VC_FilteredSocketCount()
   &SpawnRate(
    SpawnRate=/VehicleCosmetics/FX/Niagara/Modules/VC_Mask_Floats(Coro Value:NiagaraFloat=600, FrontEnd Value:NiagaraFloat=200, FN Value:NiagaraFloat=200)
    Use Spawn Probability=false
   )
  ]
  particleSpawnScript [
   &VC_PaintedVariantsOverride(
    BurntSienna=<rgba 0.0648 0.0203 0.008 1>
    Cobalt=<rgba 0.0409 0.0513 0.5647 1>
    Crimson=<rgba 0.2918 0.004 0 1>
    DefaultColor=<rgba 0.3906 0.2437 0.1624 1>
    ForestGreen=<rgba 0 0.1651 0.0086 1>
    Grey=<rgba 0.1651 0.1651 0.1651 1>
    Lime=<rgba 0.3278 0.6038 0.0529 1>
    Orange=<rgba 0.5333 0.1714 0 1>
    Pink=<rgba 0.4969 0.0976 0.3515 1>
    Purple=<rgba 0.1356 0 0.3325 1>
    Saffron=<rgba 0.5149 0.3916 0 1>
    SkyBlue=<rgba 0 0.3372 0.5029 1>
    TitaniumWhite=<rgba 0.4969 0.4969 0.4969 1>
   )
   &InitializeParticle(
    Color Mode=::NewEnumerator0
    Lifetime=&VC_Mask_Floats(Coro Value:NiagaraFloat=0.4, FrontEnd Value:NiagaraFloat=0.5, FN Value:NiagaraFloat=0.5)
    Sprite Rotation Mode=::NewEnumerator2
    Sprite Size Mode=::NewEnumerator1
    Uniform Sprite Size Max=5
    Uniform Sprite Size Min=3.5
    UsePositionOffset=false
    Write Material Random=true
   )
   &PartitionParticles()
   &SkeletalMeshLocation(
    Bone / Socket Index=&Add_Integer(A=$Particles.PartitionParticles.Partition, B=4)
    Bone / Socket Sampling Mode=::NewEnumerator5
    MaskSpawnGroup=false
    Mesh Sampling Type=::NewEnumerator6
    Sampled Position Offset=&AddVectorToPosition(Position=<-50 0 -25>, Vector=$User.DTCarOffset001)
    Skeletal Mesh=$System.SkeletalMesh
   )
   &AddVelocity(Velocity=&VC_Mask_Vectors(Coro Value:Vector3f=<-1200 0 0>, FrontEnd Value:Vector3f=<-2000 0 0>, FN Value:Vector3f=<-2000 0 0>))
   /VehicleCosmetics/FX/VC_FXParents/Modules/DLR_ProductPreviews(GameModeID:NiagaraInt32=$System.GameModeID, bPreviewVehicleSpeed:NiagaraBool=$System.bPreviewVehicleSpeed)
  ]
  particleUpdateScript [
   &ParticleState()
   &ScaleSpriteSize(
    Scale Sprite Size Mode=::NewEnumerator0
    Uniform Scale Factor=&FloatFromCurve(FloatCurve=DataInterface() {propertyValues {Curve {keys [<at 0 8>,<at 1 22>]}}})
   )
   &Drag(
    Drag=&FloatFromCurve(
     FloatCurve=DataInterface() {propertyValues {Curve {
        defaultValue=1
        keys [
         {arriveTangent=0.0291, interpMode=RCIM_Cubic, leaveTangent=0.003, tangentMode=RCTM_Break, time=0, value=0}
         {interpMode=RCIM_Cubic, tangentMode=RCTM_User, time=1, value=1}
        ]
       }}}
     Scale Curve=8
    )
    Use Linear Drag=true
    Use Rotational Drag=true
   ) {enabled=false}
   &InheritSourceMovement(
    Applied Position Scale=&VectorFromFloat(Value=&FloatFromCurve(FloatCurve=DataInterface() {propertyValues {Curve {keys [
          {interpMode=RCIM_Cubic, tangentMode=RCTM_Break, time=0, value=1}
          {arriveTangent=-2.9025, interpMode=RCIM_Cubic, tangentMode=RCTM_Break, time=1, value=0}
         ]}}}))
    ApplyForce=false
    ApplyPosition=true
    ApplyPositionFalloff=false
   ) {enabled=false}
   &InheritSourceMovement(
    Applied Position Scale=&Multiply_VectorByFloat(
     Vector=<1 1 1>
     Float=&FloatFromCurve(FloatCurve=DataInterface() {propertyValues {Curve {
         defaultValue=1
         keys [
          {arriveTangent=-0.0429, interpMode=RCIM_Cubic, leaveTangent=-0.0429, tangentMode=RCTM_User, time=0.05, value=1}
          {arriveTangent=-1.1178, interpMode=RCIM_Cubic, leaveTangent=-1.1178, tangentMode=RCTM_User, time=1, value=0}
         ]
        }}})
    )
    ApplyForce=false
    ApplyPosition=true
    ApplyPositionFalloff=false
   ) {module=InheritSourceMovement001}
   &SubUVAnimation(
    "Sprite Renderer "=DataInterface() {propertyValues {SpriteRenderer=&self:NE_CS_MainTrail_0.NiagaraSpriteRendererProperties_1}}
    Start Frame Offset=1
    UseEndFrame=false
    UseStartFrame=false
   ) {enabled=false}
   &ScaleColor(
    Scale Alpha=&FloatFromCurve(
     FloatCurve=DataInterface() {propertyValues {Curve {keys [
         {arriveTangent=2.3917, interpMode=RCIM_Cubic, leaveTangent=2.3917, tangentMode=RCTM_User, time=0, value=0.5}
         {arriveTangent=1.25, time=0.2, value=2}
         {arriveTangent=0.5269, interpMode=RCIM_Cubic, leaveTangent=0.5269, tangentMode=RCTM_User, time=0.5, value=2}
         {arriveTangent=-12.9999, interpMode=RCIM_Cubic, leaveTangent=-12.9999, tangentMode=RCTM_User, time=0.9, value=0}
        ]}}}
     Scale Curve=0.8
    )
    Scale RGB=&VectorFromFloat(Value=2)
    ScaleA=true
    ScaleRGB=true
   )
   &DynamicMaterialParameters(
    Index 0 Param 2=8
    Index 0 Param 3=&FloatFromCurve(FloatCurve=DataInterface() {propertyValues {Curve {
        defaultValue=1
        keys [
         {arriveTangent=4.1785, interpMode=RCIM_Cubic, leaveTangent=4.1785, tangentMode=RCTM_User, time=0, value=1}
         {arriveTangent=0.0224, interpMode=RCIM_Cubic, leaveTangent=0.0224, tangentMode=RCTM_User, time=1, value=64}
        ]
       }}})
    Module.Index0Float0Write=true
    Module.Index0Float1Write=true
    Module.Index0Float2Write=true
    Module.Index0Float3Write=true
   )
   &SolveForcesAndVelocity(Clamp Velocity=false, Limit Acceleration=false)
  ]
  rendererClasses [&NiagaraSpriteRendererProperties]
  renderers [{rendererClass=&NiagaraSpriteRendererProperties}]
 }
 SmokeWheel {
  emitterUpdateScript [
   &EmitterState()
   &SpawnRate(SpawnRate=30, Use Spawn Probability=false)
  ]
  particleSpawnScript [
   &VC_PaintedVariantsOverride(
    BurntSienna=<rgba 0.0648 0.0203 0.008 1>
    Cobalt=<rgba 0.0409 0.0513 0.5647 1>
    Crimson=<rgba 0.2918 0.004 0 1>
    DefaultColor=<rgba 0.3968 0.2462 0.1651 1>
    ForestGreen=<rgba 0 0.1651 0.0086 1>
    Grey=<rgba 0.1651 0.1651 0.1651 1>
    Lime=<rgba 0.3278 0.6038 0.0529 1>
    Orange=<rgba 0.5333 0.1714 0 1>
    Pink=<rgba 0.4969 0.0976 0.3515 1>
    Purple=<rgba 0.1356 0 0.3325 1>
    Saffron=<rgba 0.5149 0.3916 0 1>
    SkyBlue=<rgba 0 0.3372 0.5029 1>
    TitaniumWhite=<rgba 0.4969 0.4969 0.4969 1>
   )
   &InitializeParticle(
    Color Mode=::NewEnumerator0
    Lifetime Max=0.8
    Lifetime Min=0.8
    Lifetime Mode=::NewEnumerator1
    Sprite Rotation Mode=::NewEnumerator2
    Sprite Size Mode=::NewEnumerator3
    Uniform Sprite Size=80
    UsePositionOffset=false
    Write Material Random=true
   )
   &PartitionParticles()
   &SkeletalMeshLocation(
    Bone / Socket Index=&Add_Integer(A=$Particles.PartitionParticles.Partition, B=4)
    Bone / Socket Sampling Mode=::NewEnumerator5
    Interpolate Mesh Sampling=false
    MaskSpawnGroup=false
    Mesh Sampling Type=::NewEnumerator6
    Sampled Position Offset=&AddVectorToPosition(Position=$User.DTCarOffset, Vector=<-45 0 -15>)
    Skeletal Mesh=$System.SkeletalMesh
   )
   &AddVelocity(Velocity=&MaskVectorBySpawnGroup(Masked Vector=<-200 0 -10>, Passthrough Vector=<-300 1 -10>))
   &Color(Color=<rgba 0.3968 0.2462 0.1651 1>) {enabled=false}
  ]
  particleUpdateScript [
   &ParticleState()
   &SubUVAnimation(
    "Sprite Renderer "=DataInterface() {propertyValues {SpriteRenderer=&self:SmokeWheel_0.NiagaraSpriteRendererProperties_0}}
    UseEndFrame=false
    UseStartFrame=false
   ) {enabled=false}
   &ScaleSpriteSize(Uniform Curve Sprite Scale=DataInterface() {propertyValues {Curve {defaultValue=0, keys [{tangentMode=RCTM_User, time=0, value=1.3}, {tangentMode=RCTM_User, time=0.8, value=0}]}}})
   &SolveForcesAndVelocity(Clamp Velocity=false, Limit Acceleration=false)
   &ScaleColor(
    Scale Alpha=&FloatFromCurve(
     FloatCurve=DataInterface() {propertyValues {Curve {keys [
         <at 0 0.5 RCIM_Cubic>
         {arriveTangent=1.25, time=0.3, value=2}
         {arriveTangent=0.0312, interpMode=RCIM_Cubic, leaveTangent=0.0312, tangentMode=RCTM_User, time=0.75, value=2}
         <at 1 0 RCIM_Cubic>
        ]}}}
     Scale Curve=0.8
    )
    Scale RGB=&VectorFromFloat(Value=1.8)
    ScaleA=true
    ScaleRGB=true
   )
   &DynamicMaterialParameters(
    Index 0 Param 2=8
    Index 0 Param 3=&FloatFromCurve(FloatCurve=DataInterface() {propertyValues {Curve {keys [<at 0 1>,<at 1 64>]}}})
    Module.Index0Float0Write=true
    Module.Index0Float1Write=true
    Module.Index0Float2Write=true
    Module.Index0Float3Write=true
   )
  ]
  rendererClasses [&NiagaraSpriteRendererProperties]
  renderers [{rendererClass=&NiagaraSpriteRendererProperties}]
 }
 WheelWrap {
  emitterUpdateScript [
   &EmitterState(
    Inactive Response=::NewEnumerator1
    Life Cycle Mode=::NewEnumerator1
    Loop Behavior=::NewEnumerator1
    Loop Duration Mode=::NewEnumerator1
    UseLoopDelay=false
   )
   &SpawnBurst_Instantaneous(Spawn Count=2, Use Spawn Probability=false, UseLoopCountLimit=false)
  ]
  particleSpawnScript [
   &VC_PaintedVariantsOverride(
    Black=<rgba 0 0 0 0.5565>
    BurntSienna=<rgba 0.0625 0.0196 0.0077 0.513>
    Cobalt=<rgba 0.1882 0.1957 0.5647 0.3304>
    Crimson=<rgba 0.4635 0 0.0126 0.287>
    DefaultColor=<rgba 0.5365 0.3772 0.2794 0.2609>
    ForestGreen=<rgba 0 0.2083 0.0105 0.3304>
    Grey=<rgba 0.2031 0.2031 0.2031 0.2522>
    Lime=<rgba 0.2943 0.5885 0 0.3391>
    Orange=<rgba 0.6724 0.2346 0 0.2957>
    Pink=<rgba 0.5052 0.0605 0.3435 0.2435>
    Purple=<rgba 0.2095 0 0.5104 0.4435>
    Saffron=<rgba 1 0.8668 0 0.3652>
    SkyBlue=<rgba 0 0.6753 1 0.3478>
    TitaniumWhite=<rgba 1 1 1 0.3913>
   )
   &InitializeParticle(
    Color Mode=::NewEnumerator0
    Lifetime=0.3
    Mesh Scale Mode=::NewEnumerator3
    Sprite Rotation Mode=::NewEnumerator2
    Sprite Size Mode=::NewEnumerator3
    UsePositionOffset=false
    Write Material Random=true
   )
   &PartitionParticles()
   &SkeletalMeshLocation(
    Bone / Socket Index=&Add_Integer(A=$Particles.PartitionParticles.Partition, B=4)
    Bone / Socket Sampling Mode=::NewEnumerator5
    Interpolate Mesh Sampling=false
    MaskSpawnGroup=false
    Mesh Sampling Type=::NewEnumerator6
    Skeletal Mesh=$System.SkeletalMesh
   )
  ]
  particleUpdateScript [
   &FN_DisableTireFX(
    Mask Index Attribute:NiagaraInt32=$Particles.PartitionParticles.Partition
    Mask Index FL=2
    Mask Index FR=3
    Mask Index RL=0
    Mask Index RR=1
   )
   &VC_WheelWrapScale(Wheel Index Attribute:NiagaraInt32=&Add_Integer(A=$Particles.PartitionParticles.Partition, B=2))
   &ParticleState(Kill Particles When Lifetime Has Elapsed=false)
   &UpdateMeshOrientation(Rotation Coordinate Space=::NewEnumerator2, Rotation Vector=<0 1 0>) {enabled=false}
   &SkeletalMeshLocation(
    Bone / Socket Index=&Add_Integer(A=$Particles.PartitionParticles.Partition, B=4)
    Bone / Socket Sampling Mode=::NewEnumerator5
    MaskSpawnGroup=false
    Mesh Sampling Type=::NewEnumerator6
    Skeletal Mesh=$System.SkeletalMesh
   ) {module=SkeletalMeshLocation001}
   &SolveForcesAndVelocity(Clamp Velocity=false, Limit Acceleration=false)
   &DynamicMaterialParameters(
    Index 2 Param 1=1
    Index 2 Param 2=&Add_Float(
     A=&FloatFromCurve(
      FloatCurve=DataInterface() {propertyValues {Curve {keys [{tangentMode=RCTM_User, time=0, value=0}, {tangentMode=RCTM_User, time=1, value=1}]}}}
      Scale Curve=0.6
     )
     B=&RandomRangeFloat(Minimum=0, Maximum=3)
    )
    Index 2 Param 3=&RemapRange(Input Value To Remap=$Particles.NormalizedAge, Input Min=0.5, Input Max=1, Output Min=1, Output Max=6)
    Index 2 Param 4=&RandomRangeFloat(Minimum=1, Maximum=1)
    Module.Index2Float0Write=true
    Module.Index2Float1Write=false
    Module.Index2Float2Write=false
    Module.Index2Float3Write=true
    Write Parameter Index 0=::NewEnumerator0
    Write Parameter Index 2=::NewEnumerator1
   )
   /SetVariables(Particles.Color Llanata:LinearColor=<rgba 0.3231 0.1413 0.0296 0.1725>) {enabled=false, isSetParameters=true}
  ]
  rendererClasses [&NiagaraMeshRendererProperties]
  renderers [
   {rendererClass=&NiagaraMeshRendererProperties, rendererIndex=0}
   {rendererClass=&NiagaraMeshRendererProperties, rendererIndex=1}
  ]
 }
}
summary {
 systemName=NS_DriftTrail_Smoke
 emitters [
  {emitterName=NE_CS_MainTrail, rendererClasses [&NiagaraSpriteRendererProperties]}
  {emitterName=WheelWrap, rendererClasses [&NiagaraMeshRendererProperties]}
  {emitterName=SmokeWheel, rendererClasses [&NiagaraSpriteRendererProperties]}
 ]
 userVariables [
  User.bIsFrontendPreview {type=NiagaraBool}
  User.bIsPainted {type=NiagaraBool}
  User.bPreviewVehicleSpeed {type=NiagaraBool}
  User.ColorVariantID {type=NiagaraInt32}
  User.ColorWhite {type=NiagaraFloat, defaultValue=0, description=""}
  User.CoreIntensity {type=NiagaraFloat, defaultValue=1, description=""}
  User.DTCarOffset {type=NiagaraPosition, description="", defaultValue {}}
  User.DTCarOffset001 {type=Vector3f, description="", defaultValue {struct=Vector3f, value=<0 0 0>}}
  User.GameModeID {type=NiagaraInt32, defaultValue=3}
  User.InColorIntensity {type=NiagaraFloat, defaultValue=1, description=""}
  User.offset {type=Vector3f, description="offset from wheels(set through cosmetic template)", defaultValue {struct=Vector3f, value=<0 0 0>}}
  User.OutColorIntensity {type=NiagaraFloat, defaultValue=1, description=""}
  User.PaintedColor {type=LinearColor, description="", defaultValue {struct=LinearColor, value=<rgba 1 1 1 1>}}
  User.PaintedScalar {type=NiagaraFloat, defaultValue=1, description=""}
  User.RibbonCoreIntensity {type=NiagaraFloat, defaultValue=10, description=""}
  User.RibbonMidIntensity {type=NiagaraFloat, defaultValue=5, description=""}
  User.RibbonOutIntensity {type=NiagaraFloat, defaultValue=2, description=""}
  User.RingsIntensityB {type=NiagaraFloat, defaultValue=5, description=""}
  User.RingsIntensityG {type=NiagaraFloat, defaultValue=5, description=""}
  User.RingsIntensityR {type=NiagaraFloat, defaultValue=5, description=""}
  User.RingsOpacity {type=NiagaraFloat, defaultValue=1, description=""}
  User.SingleColorIntensity {type=NiagaraFloat, defaultValue=1, description=""}
  User.VehicleFlags {type=NiagaraInt32, defaultValue=806289408}
  User.Wheel_Radius_Array {
   type=NiagaraDataInterfaceArrayFloat
   defaultValue {
    struct=NiagaraExt_VariableValue_DataInterface
    value {
     dataInterface=&self:NiagaraDataInterfaceArrayFloat_0
     dataInterfaceClass=&NiagaraDataInterfaceArrayFloat
    }
   }
  }
  User.Wheel_Width_Array {
   type=NiagaraDataInterfaceArrayFloat
   defaultValue {
    struct=NiagaraExt_VariableValue_DataInterface
    value {
     dataInterface=&self:NiagaraDataInterfaceArrayFloat_1
     dataInterfaceClass=&NiagaraDataInterfaceArrayFloat
    }
   }
  }
  User.WheelSize {type=NiagaraFloat, defaultValue=0, description=""}
  User.WheelTouchingArray {
   type=NiagaraDataInterfaceArrayFloat
   defaultValue {
    struct=NiagaraExt_VariableValue_DataInterface
    value {
     dataInterface=&self:NiagaraDataInterfaceArrayFloat_2
     dataInterfaceClass=&NiagaraDataInterfaceArrayFloat
    }
   }
  }
 ]
}
